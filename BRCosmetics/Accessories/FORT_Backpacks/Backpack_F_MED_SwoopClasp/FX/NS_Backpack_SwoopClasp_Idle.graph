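# Idle Niagara glow for the SwoopClasp backpack cosmetic, built from one emitter with sprite and light renderers.
# Burst-spawned particles persist, with spawn groups choosing sprite size, facing and alignment to form a multidirectional glow, plus per-group curve-driven alpha.
# User.ColorIndex blends sprite and light color from teal to purple, tinted by System.TODColor.
# Brightness scales with time of day, and sprite night brightness also varies by quality level.
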NiagaraSystem /BRCosmetics/Accessories/FORT_Backpacks/Backpack_F_MED_SwoopClasp/FX/NS_Backpack_SwoopClasp_Idle
dependencies {
 usedDataInterfaces [&NiagaraDataInterfaceCurve, &NiagaraDataInterfaceColorCurve]
 usedDynamicInputs [&ColorFromCurve]
 usedModules [
  &SystemState
  &self:SystemSpawnScript.SystemScriptSource.SystemScriptGraph.NiagaraNodeAssignment_0.SetVariables_#1
 ]
}
emitters {Multidirectional_Glow {
  emitterUpdateScript [
   &EmitterState(
    Inactive Response=::NewEnumerator1
    Life Cycle Mode=::NewEnumerator1
    Loop Behavior=::NewEnumerator1
    Loop Duration Mode=::NewEnumerator1
    UseLoopDelay=false
   )
   &SpawnBurst_Instantaneous(Spawn Group=1, Use Spawn Probability=false, UseLoopCountLimit=false) {module=SpawnBurst_Instantaneous001}
   &SpawnBurst_Instantaneous(Spawn Count=3, Use Spawn Probability=false, UseLoopCountLimit=false)
  ]
  particleSpawnScript [
   /SetVariables(Particles.VisibilityTag=$Emitter.SpawnGroup) {isSetParameters=true}
   &FN_SetSpawnGroup(Number of Spawn Groups=3)
   &InitializeParticle(
    Color=<rgba 1 1 1 0.35>
    Position Offset=<2.5 -5 -3.5>
    Sprite Size=&MergeHelper_Vector2Switch(0=<10 35>, 1=<16 27.5>, 2=<10 35>, 3=?, 4=?, Index=$Particles.SpawnGroup, Num=::NewEnumerator1)
    Sprite Size Mode=::NewEnumerator4
    UsePositionOffset=true
    Write Material Random=true
   )
   &CameraOffset(Camera Offset Amount=5)
   &SpriteFacingAndAlignment(
    Sprite Alignment=&MergeHelper_VectorSwitch(0=<1 0 0>, 1=<0 0 1>, 2=<1 0 0>, 3=?, Index=$Particles.SpawnGroup, Num=::NewEnumerator1)
    Sprite Facing=&MergeHelper_VectorSwitch(0=<1.5 0 1>, 1=<0 1 0>, 2=<-1.5 0 1>, 3=?, Index=$Particles.SpawnGroup, Num=::NewEnumerator1)
    WriteAlignment=true
    WriteFacing=true
   )
  ]
  particleUpdateScript [
   &ParticleState(Kill Particles When Lifetime Has Elapsed=false)
   &ScaleColor(
    Scale Alpha=&FloatFromCurve(
     FloatCurve=DataInterface() {propertyValues {Curve {keys [<at 0 0.8 RCIM_Cubic>,<at 0.2946 1.0141 RCIM_Cubic>,<at 1 0.8 RCIM_Cubic>]}}}
     Scale Curve=&MergeHelper_FloatSwitch(0=0.6, 1=1.75, 2=0.6, 3=?, Index=$Particles.SpawnGroup, Num=::NewEnumerator1)
    )
    Scale RGB=&Multiply_VectorByFloat(
     Vector=&Multiply_Vector(
      A=&MakeVectorFromLinearColorRGB(Color=$System.TODColor)
      B=&MakeVectorFromLinearColorRGB(Color=&LerpLinearColors(StartColor=<rgba 0.0515 1 0.7773 1>, EndColor=<rgba 0.3649 0.0515 1 1>, LerpFactor=$User.ColorIndex))
     )
     Float=&FNTimeOfDayScalar(
      Night Scale=&MergeHelper_FloatSwitch(0=2, 1=2.5, 2=5, 3=5, Index=$Engine.QualityLevel, Num=::NewEnumerator2)
      Morning Scale=5
      Day Scale=1.25
      Evening Scale=4
     )
    )
    ScaleA=true
    ScaleRGB=true
   )
   &FN_Light_Attributes(
    Color=&MultiplyLinearColor(
     Linear Color=&LerpLinearColors(StartColor=<rgba 0.0515 1 0.7773 1>, EndColor=<rgba 0.3649 0.0515 1 1>, LerpFactor=$User.ColorIndex)
     Scale Factor=&MakeVector4_FromLinearColor(LinearColor=$System.TODColor)
    )
    Create Color Attribute=true
    Create Exponent Attribute=true
    Create Position Attribute=true
    Create Radius Attribute=true
    Exponent=2
    Position=&AddVectorToPosition(Position=$Engine.Emitter.SimulationPosition, Vector=<0 -15 0>)
    Radius=30
    Scale Color=&Multiply_Float(
     A=5
     B=&FNTimeOfDayScalar(Night Scale=&Lerp_Float(A=1, B=2.5, Alpha=$User.ColorIndex), Morning Scale=1, Evening Scale=1)
    )
   )
  ]
  rendererClasses [&NiagaraSpriteRendererProperties, &NiagaraLightRendererProperties]
  renderers [
   {rendererClass=&NiagaraSpriteRendererProperties}
   {rendererClass=&NiagaraLightRendererProperties, rendererIndex=1}
  ]
 }}
summary {
 systemName=NS_Backpack_SwoopClasp_Idle
 emitters [{
   emitterName=Multidirectional_Glow
   rendererClasses [&NiagaraSpriteRendererProperties, &NiagaraLightRendererProperties]
  }]
 userVariables [User.ColorIndex {type=NiagaraFloat, defaultValue=0, description=""}]
}
