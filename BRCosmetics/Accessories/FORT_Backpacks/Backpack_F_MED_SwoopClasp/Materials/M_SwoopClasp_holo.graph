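# Translucent, two-sided, unlit material for the holographic screen on the SwoopClasp back bling, usable on skeletal meshes.
# An animated flipbook noise, tinted and masked by a TechArt mask, adds to an emissive texture with optional color override and time-driven pulsing elements staggered by mask value.
# Opacity derives from the mask and a minimum value, and panning screen-static samples feed named reroutes for subtle static.
# The result passes through the shared accessory TechArt overrides function.
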Material /BRCosmetics/Accessories/FORT_Backpacks/Backpack_F_MED_SwoopClasp/Materials/M_SwoopClasp_holo
graph {
 properties {
  bOutputTranslucentVelocity=true
  bUseMaterialAttributes=true
  bUsedWithMeshDeformer=true
  bUsedWithMorphTargets=true
  bUsedWithSkeletalMesh=true
  blendMode=BLEND_Translucent
  pixelDepthOffsetMode=PDOM_Legacy
  shadingModel=MSM_Unlit
  twoSided=true
  usedShadingModels=Unlit
 }
 nodes [
  n1 = VectorParameter() {group=ColorShading, parameterName=NoiseColor, defaultValue=<rgba 0 1 0.4822 2>}
  n2 = Frac(None=Multiply(A=Time(), B=ScalarParameter() {defaultValue=0.2, parameterName=NoiseSpeed}))
  n3 = Constant() {r=8}
  n4 = TextureObject() {texture=/BRCosmetics/Accessories/FORT_Backpacks/Backpack_F_MED_SwoopClasp/Textures/T_Swoop_Noise}
  n5 = Multiply(A=TextureCoordinate(), B=ScalarParameter() {defaultValue=8, parameterName=NoiseTiling})
  n6 = TextureSampleParameter2D() {
   group=ColorShading
   parameterName=TechArt Mask
   samplerType=SAMPLERTYPE_Masks
   texture=/BRCosmetics/Accessories/FORT_Backpacks/Backpack_F_MED_SwoopClasp/Textures/T_SwoopClasp_Screen_TechArtMask
  }
  n7 = Multiply(
   A=Multiply(
    A=Multiply(A=@n1.RGB, B=@n1.RGB)
    B=MaterialFunctionCall("Animation  Phase (0-1)"=@n2, Number of Rows=@n3, Number of Columns=@n3, Texture=@n4, UVs=@n5) {materialFunction=&FlipBook}.Result
   )
   B=OneMinus(None=@n6.R)
  )
  n8 = TextureSampleParameter2D() {
   group=ColorShading
   parameterName=EmssiveColor
   texture=/BRCosmetics/Accessories/FORT_Backpacks/Backpack_F_MED_SwoopClasp/Textures/T_SwoopClasp_Backpack_E
  }
  n9 = StaticSwitchParameter(
   True=LinearInterpolate(
    A=@n8.RGB
    B=VectorParameter() {group=ColorShading, parameterName=OverrideColor, defaultValue=<rgba 0 0.4031 0.1694 1>}.RGB
    Alpha=NamedRerouteUsage()
   )
   False=@n8.RGB
  ) {group=ColorShading, parameterName=UseOverrideColor}
  n10 = NamedRerouteUsage()
  n11 = Add(A=Saturate(None=Subtract(A=@n10) {constB=0.1}), B=Multiply(A=Time(), B=ScalarParameter() {defaultValue=0.5, parameterName=ELmentSpeed}))
  n12 = LinearInterpolate(
   A=ScalarParameter() {defaultValue=2.5, parameterName=ElementIntenisty_LOW}
   B=ScalarParameter() {defaultValue=12, parameterName=ElementIntenisty_High}
   Alpha=SmoothStep(Value=Frac(None=@n11)) {constMin=0.5}
  )
  n13 = ScalarParameter() {defaultValue=1, parameterName=FrontendEmissiveIntensity}
  n14 = LinearInterpolate(
   A=Multiply(A=@n9, B=ScalarParameter() {defaultValue=1.3, group=ColorShading, parameterName=EmissiveIntensity})
   B=Multiply(A=@n9, B=Multiply(A=@n12, B=@n13))
   Alpha=Multiply(A=NamedRerouteUsage(), B=Ceil(None=@n10))
  )
  n15 = Add(A=LinearInterpolate(A=ScalarParameter() {defaultValue=0.15, parameterName=MinOpacity}, Alpha=@n6.R), B=NamedRerouteUsage())
  MaterialFunctionCall(In=SetMaterialAttributes!(MaterialAttributes, Emissive Color=Add(A=@n7, B=@n14), Opacity=Saturate(None=@n15)) {attributeSetTypes [#1, #2]}) {materialFunction=&MF_TechArtOverrides_accessories}
  NamedRerouteDeclaration(None=@n6.G) {name=ElementMask, nodeColor=<rgba 0.4531 0 1 1>}
  NamedRerouteDeclaration(None=@n6.R) {name=OutlineMask, nodeColor=<rgba 1 0 0.6797 1>}
  n16 = VectorParameter() {
   group=Default_Looping
   parameterName=ConstantNoise_Color
   channelNames {a=Intensity, b="", g="", r=""}
   defaultValue=<rgba 0.2039 0.9271 0.6979 1>
  }
  n17 = Multiply(A=TextureCoordinate(), B=ScalarParameter() {defaultValue=3, parameterName=LineTiling}) {constB=2}
  n18 = TextureSample(UVs=Panner(Coordinate=@n17) {speedX=1, speedY=0.03}) {
   samplerSource=SSM_Wrap_WorldGroupSettings
   samplerType=SAMPLERTYPE_Masks
   texture=/Game/Missions/Primary/RepairTheShelter/Textures/Cinematic_Screen_Static
  }
  n19 = TextureSample(UVs=Panner(Coordinate=@n17) {speedX=2, speedY=-0.05}) {
   samplerSource=SSM_Wrap_WorldGroupSettings
   samplerType=SAMPLERTYPE_Masks
   texture=&Cinematic_Screen_Static
  }
  n20 = TextureSample(UVs=Panner(Coordinate=@n17) {speedX=-31, speedY=-0.03}) {
   samplerSource=SSM_Wrap_WorldGroupSettings
   samplerType=SAMPLERTYPE_Masks
   texture=&Cinematic_Screen_Static
  }
  n21 = Add(A=Multiply(A=@n18.B) {constB=0.2}, B=Multiply(A=@n19.G, B=@n20.R))
  n22 = ScalarParameter() {defaultValue=3, group=TV Glitches, parameterName=Static Scale}
  n23 = Saturate(None=NamedRerouteDeclaration(None=Multiply(A=@n21, B=@n22)) {name=Static, nodeColor=<rgba 1 0.6562 0 1>})
  NamedRerouteDeclaration(None=Multiply(A=Multiply(A=@n16, B=@n16), B=Multiply(A=@n23) {constB=0.2, desc=constant static})) {name=StaticConstant, nodeColor=<rgba 0 0.7422 1 1>}
  NamedRerouteDeclaration(None=@n6.B) {name=OverrideMask, nodeColor=<rgba 0 0.1094 1 1>}
  NamedRerouteDeclaration(None=@n23) {name=SubtleNoiseAlpha, nodeColor=<rgba 1 0.5859 0 1>}
  NamedRerouteDeclaration(None=@n16) {name=NoiseTint, nodeColor=<rgba 0.8281 0 1 1>}
 ]
}
