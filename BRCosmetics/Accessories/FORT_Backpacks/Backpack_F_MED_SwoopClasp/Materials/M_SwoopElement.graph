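# Masked, two-sided character-parent material for the SwoopClasp backpack cosmetic, adding a colored Fresnel rim glow to the shared character shading.
# The 2023 Character Parent PreFX function feeds the OpacityMask function, whose emissive is scaled by FrontendEmissiveIntensity.
# That emissive lerps toward FresnelColorOverride through a powered, smoothstepped Fresnel, replacing rather than adding to it at the edges.
# The modified attributes then pass through the shared PostFX function.
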Material /BRCosmetics/Accessories/FORT_Backpacks/Backpack_F_MED_SwoopClasp/Materials/M_SwoopElement
graph {
 properties {
  bUseMaterialAttributes=true
  bUsedWithMeshDeformer=true
  bUsedWithMorphTargets=true
  bUsedWithSkeletalMesh=true
  blendMode=BLEND_Masked
  pixelDepthOffsetMode=PDOM_Legacy
  shadingModel=MSM_FromMaterialExpression
  twoSided=true
  usedShadingModels=Default Lit | Subsurface | Clear Coat | Subsurface Profile | Cloth
 }
 nodes [
  n1 = MaterialFunctionCall() {materialFunction=&MF_CharacterParent_2023_PreFX}
  n2 = GetMaterialAttributes(None=MaterialFunctionCall!(Result=@n1.Result) {materialFunction=/Game/Global/Materials/Character_Parent_2023/Functions/MF_CharacterParent_2023_OpacityMask}."Switch Param (False)") {attributeGetTypes [#1]}
  n3 = Multiply(A=@n2.Emissive Color, B=ScalarParameter() {defaultValue=1, parameterName=FrontendEmissiveIntensity})
  n4 = VectorParameter() {group=override, parameterName=FresnelColorOverride, defaultValue=<rgba 0 0.3958 0.2301 1>}
  n5 = SmoothStep(
   Max=ScalarParameter() {defaultValue=1, parameterName=FresnelMax}
   Value=Power(Base=Fresnel() {baseReflectFraction=0, exponent=1}, Exp=ScalarParameter() {defaultValue=2.2, group=override, parameterName=PowerFresnel})
  ) {constMax=0.7}
  MaterialFunctionCall(
   Material=SetMaterialAttributes(MaterialAttributes=@n2.MaterialAttributes, Emissive Color=LinearInterpolate(A=@n3, B=@n4.RGB, Alpha=@n5)) {attributeSetTypes [#1]}
   M=@n1.M
   Pre Skinned Local Position=@n1.M
   Distance Blend=@n1.M
   Fresnel=@n1.M
  ) {materialFunction=&MF_CharacterParent_2023_PostFX}
 ]
}
