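# Animation Blueprint for the SwoopClasp back bling, procedurally animating a six-legged bug with whiskers, plus spinning sphere and DNA bones.
# Each update, RotationLogic accumulates a randomly varied, wrapping spin, while sine waves drive whisker wiggle and leg swing.
# The swing's sign selects which alternating tripod (L1/R2/L3 or R1/L2/R3) updates its lift, resetting the other's timer.
# The AnimGraph applies these values as chained additive bone-space Transform Bone nodes; its Control Rig node is unconnected.
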AnimBlueprint /BRCosmetics/Accessories/FORT_Backpacks/Backpack_F_MED_SwoopClasp/Meshes/F_MED_SwoopClasp_Pack_AnimBP
target_skeleton=&/F_MED_SwoopClasp_Pack_Skeleton
graphs [
 AnimGraph {nodes [
   Control Rig() {alphaBoolBlend {blendOption=Linear}, controlRigAssetReference=~}
   n1 = Get Legs_2_LiftHeight() -> (Legs_2_LiftHeight:double)
   n2 = "Transform (Modify) Bone - Bone: dyn_sphere_01"!(
    ComponentPose=Local To Component(LocalPose=Input Pose())
    bAlphaBoolEnabled:bool=True
    Translation="0,0,0"
    Rotation=Get Sphere rotation() -> (Sphere rotation:Rotator)
    Scale="1,1,1"
   ) -> (Pose) {
    alphaInputType=Bool
    rotationMode=BMM_Additive
    rotationSpace=BCS_BoneSpace
    alphaBoolBlend {blendOption=Linear}
    boneToModify {boneName=dyn_sphere_01}
   }
   n3 = "Transform (Modify) Bone - Bone: dyn_DNA_01"(ComponentPose=@n2, Alpha=1, Translation="0,0,0", Rotation=Get DNArotator() -> (DNArotator:Rotator), Scale="1,1,1") {rotationMode=BMM_Additive, rotationSpace=BCS_BoneSpace, alphaBoolBlend {blendOption=Linear}, boneToModify {boneName=dyn_DNA_01}}
   n4 = Get Legs_L1_L3() -> (Legs_L1_L3:double)
   n5 = "Transform (Modify) Bone - Bone: dyn_L1"!(ComponentPose=@n3, Alpha=1, Translation="0,0,0", Rotation_Roll:float=0, Rotation_Pitch:float=0, Rotation_Yaw:float=@n4, Scale="1,1,1") -> (Pose) {
    rotationMode=BMM_Additive
    rotationSpace=BCS_BoneSpace
    alphaBoolBlend {blendInTime=0.1, blendOption=Linear, blendOutTime=0.1}
    boneToModify {boneName=dyn_L1}
   }
   n6 = "Transform (Modify) Bone - Bone: dyn_R2"!(
    ComponentPose=@n5
    Alpha=1
    Translation="0,0,0"
    Rotation_Roll:float=0
    Rotation_Pitch:float=0
    Rotation_Yaw:float=Get Legs_R2() -> (Legs_R2:double)
    Scale="1,1,1"
   ) -> (Pose) {
    rotationMode=BMM_Additive
    rotationSpace=BCS_BoneSpace
    alphaBoolBlend {blendInTime=0.1, blendOption=Linear, blendOutTime=0.1}
    boneToModify {boneName=dyn_R2}
   }
   n7 = "Transform (Modify) Bone - Bone: dyn_L3"!(ComponentPose=@n6, Alpha=1, Translation="0,0,0", Rotation_Roll:float=0, Rotation_Pitch:float=0, Rotation_Yaw:float=@n4, Scale="1,1,1") -> (Pose) {
    rotationMode=BMM_Additive
    rotationSpace=BCS_BoneSpace
    alphaBoolBlend {blendInTime=0.1, blendOption=Linear, blendOutTime=0.1}
    boneToModify {boneName=dyn_L3}
   }
   n8 = Get Legs_R1_R3() -> (Legs_R1_R3:double)
   n9 = "Transform (Modify) Bone - Bone: dyn_R1"!(ComponentPose=@n7, Alpha=1, Translation="0,0,0", Rotation_Roll:float=0, Rotation_Pitch:float=0, Rotation_Yaw:float=@n8, Scale="1,1,1") -> (Pose) {
    rotationMode=BMM_Additive
    rotationSpace=BCS_BoneSpace
    alphaBoolBlend {blendInTime=0.1, blendOption=Linear, blendOutTime=0.1}
    boneToModify {boneName=dyn_R1}
   }
   n10 = "Transform (Modify) Bone - Bone: dyn_L2"!(
    ComponentPose=@n9
    Alpha=1
    Translation="0,0,0"
    Rotation_Roll:float=0
    Rotation_Pitch:float=0
    Rotation_Yaw:float=Get Legs_L2() -> (Legs_L2:double)
    Scale="1,1,1"
   ) -> (Pose) {
    rotationMode=BMM_Additive
    rotationSpace=BCS_BoneSpace
    alphaBoolBlend {blendInTime=0.1, blendOption=Linear, blendOutTime=0.1}
    boneToModify {boneName=dyn_L2}
   }
   n11 = "Transform (Modify) Bone - Bone: dyn_R3"!(ComponentPose=@n10, Alpha=1, Translation="0,0,0", Rotation_Roll:float=0, Rotation_Pitch:float=0, Rotation_Yaw:float=@n8, Scale="1,1,1") -> (Pose) {
    rotationMode=BMM_Additive
    rotationSpace=BCS_BoneSpace
    alphaBoolBlend {blendInTime=0.1, blendOption=Linear, blendOutTime=0.1}
    boneToModify {boneName=dyn_R3}
   }
   n12 = Get Legs_1_LiftHeight() -> (Legs_1_LiftHeight:double)
   n13 = "Transform (Modify) Bone - Bone: dyn_L1"!(ComponentPose=@n11, bAlphaBoolEnabled:bool=True, Translation="0,0,0", Rotation_Roll=0, Rotation_Pitch=@n12, Rotation_Yaw=0, Scale="1,1,1") -> (Pose) {
    alphaInputType=Bool
    rotationMode=BMM_Additive
    rotationSpace=BCS_BoneSpace
    alphaBoolBlend {blendInTime=0.1, blendOption=Linear, blendOutTime=0.1}
    boneToModify {boneName=dyn_L1}
   }
   n14 = "Transform (Modify) Bone - Bone: dyn_R2"!(ComponentPose=@n13, bAlphaBoolEnabled:bool=True, Translation="0,0,0", Rotation_Roll=0, Rotation_Pitch=@n12, Rotation_Yaw=0, Scale="1,1,1") -> (Pose) {
    alphaInputType=Bool
    rotationMode=BMM_Additive
    rotationSpace=BCS_BoneSpace
    alphaBoolBlend {blendInTime=0.1, blendOption=Linear, blendOutTime=0.1}
    boneToModify {boneName=dyn_R2}
   }
   n15 = "Transform (Modify) Bone - Bone: dyn_L3"!(ComponentPose=@n14, bAlphaBoolEnabled:bool=True, Translation="0,0,0", Rotation_Roll=0, Rotation_Pitch=@n12, Rotation_Yaw=0, Scale="1,1,1") -> (Pose) {
    alphaInputType=Bool
    rotationMode=BMM_Additive
    rotationSpace=BCS_BoneSpace
    alphaBoolBlend {blendInTime=0.1, blendOption=Linear, blendOutTime=0.1}
    boneToModify {boneName=dyn_L3}
   }
   n16 = "Transform (Modify) Bone - Bone: dyn_R1"!(ComponentPose=@n15, bAlphaBoolEnabled:bool=True, Translation="0,0,0", Rotation_Roll=0, Rotation_Pitch=@n1, Rotation_Yaw=0, Scale="1,1,1") -> (Pose) {
    alphaInputType=Bool
    rotationMode=BMM_Additive
    rotationSpace=BCS_BoneSpace
    alphaBoolBlend {blendInTime=0.1, blendOption=Linear, blendOutTime=0.1}
    boneToModify {boneName=dyn_R1}
   }
   n17 = "Transform (Modify) Bone - Bone: dyn_L2"!(ComponentPose=@n16, bAlphaBoolEnabled:bool=True, Translation="0,0,0", Rotation_Roll=0, Rotation_Pitch=@n1, Rotation_Yaw=0, Scale="1,1,1") -> (Pose) {
    alphaInputType=Bool
    rotationMode=BMM_Additive
    rotationSpace=BCS_BoneSpace
    alphaBoolBlend {blendInTime=0.1, blendOption=Linear, blendOutTime=0.1}
    boneToModify {boneName=dyn_L2}
   }
   n18 = "Transform (Modify) Bone - Bone: dyn_R3"!(ComponentPose=@n17, bAlphaBoolEnabled:bool=True, Translation="0,0,0", Rotation_Roll=0, Rotation_Pitch=@n1, Rotation_Yaw=0, Scale="1,1,1") -> (Pose) {
    alphaInputType=Bool
    rotationMode=BMM_Additive
    rotationSpace=BCS_BoneSpace
    alphaBoolBlend {blendInTime=0.1, blendOption=Linear, blendOutTime=0.1}
    boneToModify {boneName=dyn_R3}
   }
   n19 = Get WhiskersMotion() -> (WhiskersMotion:double)
   n20 = "Transform (Modify) Bone - Bone: dyn_whiskers"!(ComponentPose=@n18, Alpha=1, Translation="0,0,0", Rotation_Roll:float=@n19, Rotation_Pitch:float=@n19, Rotation_Yaw:float=0, Scale="1,1,1") -> (Pose) {
    rotationMode=BMM_Additive
    rotationSpace=BCS_BoneSpace
    alphaBoolBlend {blendInTime=0.1, blendOption=Linear, blendOutTime=0.1}
    boneToModify {boneName=dyn_whiskers}
   }
   Output Pose(Result=Component To Local(ComponentPose=@n20)) {layerGroup=None}
  ]}
 EventGraph {nodes [
   n1 = Event BlueprintUpdateAnimation()
   > Set Delta Time X(Delta Time X=@n1.DeltaTimeX)
   > n2 = Sequence()
   > Set Delta time remap(Delta time remap:double=FInterpTo(Current=Get Delta time remap() -> (Delta time remap:double), Target=1, DeltaTime=0, InterpSpeed=1)) -> (Output_Get:double)
   n3 = RotationLogic!(execute=@n2.then_1, self, Spin:double=Get Spin() -> (Spin:double), Speed:double=Get Speed() -> (Speed:double)) -> (then, Output:double)
   > n4 = Set Spin(Spin:double=@n3) -> (Output_Get:double)
   > Set Sphere rotation(Sphere rotation:Rotator=MakeRotator(Roll=@n4, Pitch=@n4)) -> (Output_Get:Rotator)
   > Set!(execute, DNArotator_Roll:float=@n4, DNArotator_Pitch:float=0, DNArotator_Yaw:float=0) -> (then, Output_Get:Rotator)
   > Set AccumulateTime(AccumulateTime:double=float + float(A=Get AccumulateTime() -> (AccumulateTime:double), B=Get Delta Time X())) -> (Output_Get:double)
   n5 = Get AccumulateTime()
   n6 = Get Delta Time X()
   n7 = float * float(
    A=float * float(A=float + float(A=@n5, B=@n6), B=Get BugSpeed() -> (BugSpeed:double))
    B=float / float(A=Get Amplitude() -> (Amplitude:double), B=3)
   ) {is=title K2Node_PromotableOperator}
   > Set WhiskersMotion(WhiskersMotion:double=float * float(A="Sin (Radians)"(A=@n7), B=3)) -> (Output_Get:double)
   > Set RawSinA(RawSinA:double="Sin (Radians)"(A=Get EffectiveTimeA() -> (EffectiveTimeA:double))) -> (Output_Get:double)
   n8 = Get RawSinA() -> (RawSinA:double)
   > Set SlowFactorA(SlowFactorA:double=float - float(A=2.2, B="Absolute (Float)"(A=@n8))) -> (Output_Get:double)
   > Set DeltaA(DeltaA:double=float * float(A=Get Delta Time X(), B=6)) -> (Output_Get:double)
   > Set EffectiveTimeA(EffectiveTimeA:double=float + float(A=Get EffectiveTimeA(), B=Get DeltaA() -> (DeltaA:double))) -> (Output_Get:double)
   n9 = Get RawSinA()
   > Set PendulumResult(PendulumResult:double=float / float(A=float * float(A=@n9, B=40), B=2.5)) -> (Output_Get:double)
   > n10 = Branch(Condition="float >= float"(A=Get PendulumResult() -> (PendulumResult:double)))
   > Set Legs_1_Up(Legs_1_Up:bool=true) -> (Output_Get:bool)
   > Set Legs_2_Up(Legs_2_Up:bool=false) -> (Output_Get:bool)
   > n11 = Set Legs_L1_L3(Legs_L1_L3:double=Get PendulumResult()) -> (Output_Get:double)
   > n12 = Set Legs_R2(Legs_R2:double=float * float(A=@n11, B=-1)) -> (Output_Get:double)
   > n13 = Set Legs_L2(Legs_L2:double=@n12) -> (Output_Get:double)
   > Set Legs_R1_R3(Legs_R1_R3:double=float * float(A=@n13, B=-1)) -> (Output_Get:double)
   > n14 = Set Legs_1_Time(Legs_1_Time:double=float + float(A=Get Legs_1_Time() -> (Legs_1_Time:double), B=Get Delta Time X())) -> (Output_Get:double)
   n15 = float / float(A=@n14, B=2)
   n16 = Get LegsLiftSpeed() -> (LegsLiftSpeed:double)
   n17 = float * float(A="Sin (Radians)"(A=float * float(A=@n15, B=@n16)), B=0.5)
   > Set Legs_1_LiftHeight(Legs_1_LiftHeight:double=float * float(A=float + float(A=@n17, B=0.5), B=30)) -> (Output_Get:double)
   > Set Legs_2_Time(Legs_2_Time:double=0) -> (Output_Get:double)
   Set Legs_1_Up(execute=@n10.else, Legs_1_Up=false)
   > Set Legs_2_Up(Legs_2_Up=true)
   > n18 = Set Legs_L1_L3(Legs_L1_L3=Get PendulumResult())
   > n19 = Set Legs_R2(Legs_R2=float * float(A=@n18, B=-1))
   > n20 = Set Legs_L2(Legs_L2=@n19)
   > Set Legs_R1_R3(Legs_R1_R3=float * float(A=@n20, B=-1))
   > n21 = Set Legs_2_Time(Legs_2_Time=float + float(A=Get Legs_2_Time() -> (Legs_2_Time:double), B=Get Delta Time X()))
   n22 = float / float(A=@n21, B=2)
   n23 = Get LegsLiftSpeed()
   n24 = float * float(A="Sin (Radians)"(A=float * float(A=@n22, B=@n23)), B=0.5)
   > Set Legs_2_LiftHeight(Legs_2_LiftHeight:double=float * float(A=float + float(A=@n24, B=0.5), B=30)) -> (Output_Get:double)
   > Set Legs_1_Time(Legs_1_Time=0)
   TryGetPawnOwner()
  ]}
 RotationLogic {nodes [
   n1 = RotationLogic!() -> (then, Spin:double=~, Speed:double=~) {is=title K2Node_FunctionEntry}
   > n2 = Set Spin var(Spin var:double=@n1) -> (Output_Get:double)
   > Set Acceleration(Acceleration:double=FInterpTo(Current=Get Acceleration() -> (Acceleration:double), Target=RandomFloatInRange(Min=0, Max=0.2), DeltaTime=Get Delta Time X(), InterpSpeed=1)) -> (Output_Get:double)
   > n3 = Branch(Condition="float >= float"(A=@n2, B=720))
   > n4 = Set Spin var(Spin var=float - float(A=Get Spin var() -> (Spin var:double), B=720))
   n5 = Get Velocity mult() -> (Velocity mult:double)
   n6 = Get Acceleration()
   n7 = Get Delta time remap()
   n8 = Set Spin var(
    execute=@n3.else
    Spin var=float + float!(A=Get Spin var(), B=float * float(A=@n1.Speed, B=@n7), C:double=float * float(A=@n6, B=@n5)) -> (ReturnValue)
   )
   Return Node!(execute=[@n8.then, @n4.then], Output:double=@n8)
  ]}
]
variables [
 Acceleration {type=number}
 AccumulateTime {type=number}
 Amplitude {type=number, default=40}
 Bug moving {type=boolean}
 BugSpeed {type=number, default=0.5}
 Delta time remap {type=number}
 Delta Time X {type=number}
 DeltaA {type=number}
 DNArotator {type=Rotator}
 EffectiveTimeA {type=number}
 Gate {type=number}
 Legs_1_LiftHeight {type=number}
 Legs_1_Time {type=number}
 Legs_1_Up {type=boolean}
 Legs_2_LiftHeight {type=number}
 Legs_2_Time {type=number}
 Legs_2_Up {type=boolean}
 Legs_L1_L3 {type=number}
 Legs_L2 {type=number}
 Legs_R1_R3 {type=number}
 Legs_R2 {type=number}
 LegsLiftSpeed {type=number, default=20}
 PendulumResult {type=number}
 RawSinA {type=number}
 SlowFactorA {type=number}
 Speed {type=number, default=5}
 Sphere rotation {type=Rotator}
 Spin {type=number}
 Spin var {type=number}
 TestMult {type=number, default=2}
 Velocity mult {type=number, default=12}
 WhiskersMotion {type=number}
]
